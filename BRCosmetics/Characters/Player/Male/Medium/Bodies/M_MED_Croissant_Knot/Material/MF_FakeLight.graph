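# Material function for a character cosmetic that optionally fakes directional lighting on incoming material attributes, behind static switches.
# A biased, powered dot product of pixel world normal and a direction parameter, floored and saturated, forms the light mask.
# Switches select an additive variant writing Emissive or a darkening variant writing BaseColor, both built from BaseColor multiplication.
# An optional Distance_Blend-driven lerp varies either effect with distance.
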MaterialFunction /BRCosmetics/Characters/Player/Male/Medium/Bodies/M_MED_Croissant_Knot/Material/MF_FakeLight
graph {nodes [
  n1 = NamedRerouteUsage()
  n2 = NamedRerouteUsage()
  n3 = FunctionInput() {inputName=Material, inputType=FunctionInput_MaterialAttributes}
  FunctionOutput(None=StaticSwitchParameter(
    True=StaticSwitchParameter(True=@n1, False=@n2) {group=FakeLight, parameterName=Use AdditiveFakeLight, sortPriority=0}
    False=NamedRerouteDeclaration(None=@n3) {name=InMaterial, nodeColor=<rgba 1 0 0.4922 1>}
   ) {group=Skin FX, parameterName=Use FakeLight})
  n4 = GetMaterialAttributes(None=@n3) {attributeGetTypes [#1, #2]}
  n5 = Multiply(A=@n4.BaseColor, B=NamedRerouteUsage())
  n6 = Add(A=@n4.EmissiveColor, B=Multiply(A=ScalarParameter() {defaultValue=1, group=FakeLight, parameterName=FakeLight_AdditiveStrength}, B=@n5))
  n7 = LinearInterpolate(A=@n4.BaseColor, B=@n6, Alpha=NamedRerouteUsage())
  NamedRerouteDeclaration(None=SetMaterialAttributes!(
    MaterialAttributes=NamedRerouteUsage()
    EmissiveColor=StaticSwitchParameter(True=@n7, False=@n6) {group=FakeLight, parameterName=Use DistanceBasedFakeLight, sortPriority=0}
   ) {attributeSetTypes [#2]}) {name=Additive, nodeColor=<rgba 1 0 0.6797 1>}
  n8 = LinearInterpolate(A=@n4.BaseColor, B=@n5, Alpha=NamedRerouteUsage())
  NamedRerouteDeclaration(None=SetMaterialAttributes!(
    MaterialAttributes=NamedRerouteUsage()
    BaseColor=StaticSwitchParameter(True=@n8, False=@n5) {group=FakeLight, parameterName=Use DistanceBasedFakeLight, sortPriority=0}
   ) {attributeSetTypes [#1]}) {name=Darken, nodeColor=<rgba 1 0.4688 0 1>}
  NamedRerouteDeclaration(None=MaterialFunctionCall!(
    Blend Range=ScalarParameter() {defaultValue=3000, desc=spread, group=FakeLight, parameterName=fakeLight_Distance_blendRange}
    Start Offset=ScalarParameter() {defaultValue=-1000, desc="start (was -1500)", group=FakeLight, parameterName=fakeLight_Distance_startOffset}
   ) {materialFunction=/Engine/Functions/Engine_MaterialFunctions03/Particles/Distance_Blend}.Result) {name=DistanceBlend, nodeColor=<rgba 1 0.9375 0 1>}
  n9 = Normalize(VectorInput=VectorParameter() {group=FakeLight, parameterName=fakeLight_Direction, defaultValue=<rgba 0 0 1 0>}.RGB)
  n10 = DotProduct(A=PixelNormalWS(), B=@n9)
  n11 = Power(
   Base=Multiply(A=Add(A=@n10) {constB=0.5}, B=ScalarParameter() {defaultValue=1.5, group=FakeLight, parameterName=fakeLight_DotMult}) {constB=1.8}
   Exp=ScalarParameter() {defaultValue=2, group=FakeLight, parameterName=fakeLight_DotPow}
  ) {constExponent=3}
  n12 = ScalarParameter() {defaultValue=0.5, group=FakeLight, parameterName=fakeLight_DotMinimumValue}
  NamedRerouteDeclaration(None=Saturate(None=Max(A=@n11, B=@n12))) {name=fake top light, nodeColor=<rgba 0.6641 0 1 1>}
  Transform(None=@n9) {transformSourceType=TRANSFORMSOURCE_View}
 ]}
