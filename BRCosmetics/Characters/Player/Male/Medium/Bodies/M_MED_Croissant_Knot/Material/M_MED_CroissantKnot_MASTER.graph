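# Master material for the Croissant Knot male medium player body, built on material attributes for skeletal meshes with a subsurface skin profile.
# Shared PreFX and PostFX character functions bracket an effect chain: fake light, cloth fuzz, silk, clear coat, thin film, metal LUT and emissive fresnel.
# Static switches gate the middle five effects, with only clear coat and thin film on by default.
# A Skin FX mask texture and channel parameters support per-effect masking.
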Material /BRCosmetics/Characters/Player/Male/Medium/Bodies/M_MED_Croissant_Knot/Material/M_MED_CroissantKnot_MASTER
graph {
 properties {
  bUseMaterialAttributes=true
  bUsedWithMeshDeformer=true
  bUsedWithMorphTargets=true
  bUsedWithSkeletalMesh=true
  pixelDepthOffsetMode=PDOM_Legacy
  shadingModel=MSM_FromMaterialExpression
  subsurfaceProfile=&SS_HeroSkin_02
  usedShadingModels=Default Lit | Clear Coat | Subsurface Profile | Cloth
 }
 nodes [
  n1 = MaterialFunctionCall() {materialFunction=&BaseFN_CharacterFunction_PreFX}
  MaterialFunctionCall(AO=@n1.AO, Distance Blend=@n1.AO, Fresnel=@n1.AO, Gmap=NamedRerouteUsage(), Pre Skinned Local Position=@n1.AO, Material=NamedRerouteUsage()) {materialFunction=&BaseFN_CharacterFunction_PostFX}
  n2 = NamedRerouteUsage()
  n3 = NamedRerouteUsage()
  n4 = NamedRerouteUsage()
  n5 = NamedRerouteUsage()
  n6 = MaterialFunctionCall!(Material=NamedRerouteUsage()) {materialFunction=&/MF_FakeLight}
  n7 = StaticSwitchParameter(
   True=MaterialFunctionCall!(
    Alpha=StaticComponentMaskParameter(None=@n5) {defaultR=true, group=Skin FX, parameterName=Cloth Channel}
    Gmap=NamedRerouteUsage()
    Material=@n6.Result
   ) {materialFunction=/Game/Weapons/WeaponSkins/MaterialFunction/ClothFuzz}.Result
   False=@n6.Result
  ) {group=Skin FX, parameterName=Use Cloth fuzz}
  n8 = StaticSwitchParameter(
   True=MaterialFunctionCall!(Alpha=ChannelMaskParameter(None=@n4) {group=Skin FX, parameterName=Silk Channel}.R, Material=@n7) {materialFunction=/Game/Weapons/WeaponSkins/MaterialFunction/MF_Silk}.Result
   False=@n7
  ) {group=Skin FX, parameterName=Use Silk}
  n9 = StaticSwitchParameter(
   True=MaterialFunctionCall!(Alpha=StaticComponentMaskParameter(None=@n3) {defaultR=true, group=Skin FX, parameterName=Clear Coat Channel}, In=@n8) {materialFunction=/Game/Weapons/WeaponSkins/MaterialFunction/Clearcoat}.Result
   False=@n8
  ) {defaultValue=true, group=Skin FX, parameterName=Use Clear Coat}
  n10 = StaticSwitchParameter(
   True=MaterialFunctionCall!(Alpha=StaticComponentMaskParameter(None=@n2) {defaultR=true, group=Skin FX, parameterName=ThinFilm_Channel}, Material Attributes=@n9) {materialFunction=/Game/Weapons/WeaponSkins/MaterialFunction/ThinFilmFunction}.Result
   False=@n9
  ) {defaultValue=true, group=Skin FX, parameterName=Use Thin Film}
  n11 = MaterialFunctionCall(Mask=NamedRerouteUsage(), Materials=@n10) {materialFunction=&MF_MetalsLUT}
  NamedRerouteDeclaration(None=MaterialFunctionCall!(In=StaticSwitchParameter(True=@n11.Result, False=@n10) {group=Skin FX, parameterName=UseMetalLUT}) {materialFunction=/Game/Weapons/WeaponSkins/MaterialFunction/MF_EmissiveFresnel}.Result) {name=MaterialOut, nodeColor=<rgba 1 0 0.8203 1>}
  NamedRerouteDeclaration(None=@n1.Result) {name=MaterialIn, nodeColor=<rgba 0.2656 1 0 1>}
  NamedRerouteDeclaration(None=@n1.Gmap) {name=Gmap, nodeColor=<rgba 1 0 0.7266 1>}
  NamedRerouteDeclaration(None=TextureSample(Tex=TextureObjectParameter() {group=Skin FX, parameterName=SkinFX_Mask}).RGB) {name=SkinFX_Mask_RGB, nodeColor=<rgba 1 0 0.3047 1>}
 ]
}
