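# Animation Blueprint for the M_MED_Croissant_Knot player body cosmetic, adding deformation and jacket motion to a pose copied from its attached parent mesh.
# The copied pose runs through twist (half alpha) and sculpter Control Rigs.
# Each update maps thigh yaw deltas from ref pose, relative to pelvis, to alphas for additive offsets that shift jacket bones as legs move.
# A RigidBody node then simulates the jacket in spine_02 space with damped component motion.
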AnimBlueprint /BRCosmetics/Characters/Player/Male/Medium/Bodies/M_MED_Croissant_Knot/Meshes/M_MED_Croissant_Knot_AnimBP
target_skeleton=&/M_MED_Croissant_Knot_Skeleton
graphs [
 AnimGraph {nodes [
   n1 = Control Rig(Source=Copy Pose From Mesh() {bUseAttachedParent=true}, Alpha=0.5) {
    alphaBoolBlend {blendOption=Linear}
    controlRigAssetReference {blueprintRigClass=&/M_MED_Croissant_Deform_CtrlRig_twists.M_MED_Croissant_Deform_CtrlRig_twists_C}
   }
   n2 = Control Rig!(
    Source=@n1
    Alpha
    KneeRot_fr_l_114deg:Vector
    KneeTrans_fr_l_114deg:Vector="0.258,-3.591,0.009"
    KneeScale_fr_l_114deg:Vector="1,1,1"
    KneeScale_fr_l_130deg:Vector="1,1,1"
    KneeRot_fr_l_130deg:Vector="0,0,-51.2"
    KneeTrans_fr_l_130deg:Vector="0.258,-3.591,0.009"
   ) -> (Pose) {
    alphaBoolBlend {blendOption=Linear}
    controlRigAssetReference {blueprintRigClass=&/M_MED_Croissant_Deform_CtrlRig_sculpters.M_MED_Croissant_Deform_CtrlRig_sculpters_C}
   }
   n3 = Get Thigh Up L() -> (Thigh Up L:double)
   n4 = "Transform (Modify) Bone - Bone: dyn_jacket_l_fr_MSTR"(ComponentPose=Local To Component(LocalPose=@n2), Alpha=@n3, Translation="0,-3,3", Rotation="25,-35,-60", Scale="1,1,1") {
    rotationMode=BMM_Additive
    rotationSpace=BCS_BoneSpace
    translationMode=BMM_Additive
    translationSpace=BCS_BoneSpace
    alphaBoolBlend {blendOption=Linear}
    boneToModify {boneName=dyn_jacket_l_fr_MSTR}
    rotation=<rot 25 -35 -60>
    translation {y=-3, z=3}
   }
   n5 = "Transform (Modify) Bone - Bone: dyn_jacket_l_MSTR"(ComponentPose=@n4, Alpha=@n3, Translation="0,0,2", Rotation="15,0,0", Scale="1,1,1") {
    rotationMode=BMM_Additive
    rotationSpace=BCS_BoneSpace
    translationMode=BMM_Additive
    translationSpace=BCS_BoneSpace
    alphaBoolBlend {blendOption=Linear}
    boneToModify {boneName=dyn_jacket_l_MSTR}
    rotation {pitch=15}
    translation {z=2}
   }
   n6 = Get Thigh Up R() -> (Thigh Up R:double)
   n7 = "Transform (Modify) Bone - Bone: dyn_jacket_r_fr_MSTR"(ComponentPose=@n5, Alpha=@n6, Translation="0,3,3", Rotation="25,35,60", Scale="1,1,1") {
    rotationMode=BMM_Additive
    rotationSpace=BCS_BoneSpace
    translationMode=BMM_Additive
    translationSpace=BCS_BoneSpace
    alphaBoolBlend {blendOption=Linear}
    boneToModify {boneName=dyn_jacket_r_fr_MSTR}
    rotation=<rot 25 35 60>
    translation {y=3, z=3}
   }
   n8 = "Transform (Modify) Bone - Bone: dyn_jacket_r_MSTR"(ComponentPose=@n7, Alpha=@n6, Translation="0,0,0", Rotation="15,0,0", Scale="1,1,1") {
    rotationMode=BMM_Additive
    rotationSpace=BCS_BoneSpace
    translationMode=BMM_Additive
    translationSpace=BCS_BoneSpace
    alphaBoolBlend {blendOption=Linear}
    boneToModify {boneName=dyn_jacket_r_MSTR}
    rotation {pitch=15}
   }
   n9 = RigidBody(ComponentPose=@n8) {
    bOverrideWorldGravity=true
    overridePhysicsAsset=&/M_MED_Croissant_Knot_Jacket_Physics
    simulationSpace=BaseBoneSpace
    alphaBoolBlend {blendOption=Linear}
    baseBoneRef {boneName=spine_02}
    componentAppliedLinearAccClamp=<3000 3000 3000>
    componentLinearAccScale=<0.2 0.2 0.4>
    componentLinearVelScale=<0.1 0.1 0.2>
    overrideWorldGravity {z=-980}
   }
   Output Pose(Result=Component To Local(ComponentPose=@n9)) {layerGroup=None}
  ]}
 EventGraph {nodes [
   Event BlueprintUpdateAnimation()
   n1 = BreakTransform(InTransform=GetDeltaTransformFromRefPose(self=GetOwnerSkeletalMeshComponent(), BoneName=thigh_l, BaseName=pelvis))
   > Set Thigh Up L(Thigh Up L:double=MapRangeClamped(Value=BreakRotator(InRot=@n1.Rotation).Yaw, InRangeA=-20, InRangeB=-100)) -> (Output_Get:double)
   n2 = BreakTransform(InTransform=GetDeltaTransformFromRefPose(self=GetOwnerSkeletalMeshComponent(), BoneName=thigh_r, BaseName=pelvis))
   > Set Thigh Up R(Thigh Up R:double=MapRangeClamped(Value=BreakRotator(InRot=@n2.Rotation).Yaw, InRangeA=-20, InRangeB=-100)) -> (Output_Get:double)
  ]}
]
variables [
 Thigh Up L {type=number}
 Thigh Up R {type=number}
]
